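# Conversation database for the ChillCat NPC from NPCLibrary, defining what it says when a player talks to it.
# The entry point leads to one Speech node with a default line and contextual overrides for a first-ever conversation or specific player character cosmetics.
# The first-meeting line takes its priority from a curve row, while cosmetic lines clear that curve.
# Speech then branches through link nodes to the shared NPCLibrary Quests and BasicServices conversation entries.
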ConversationDatabase /NPCLibrary/NPCs/ChillCat/NPCLibrary_Conversation_ChillCat
entry_points [{tag=Conversation.Entry.ChillCat, destinations [#1]}]
nodes [
 ConversationEntryPointNode_7 {
  id=#1
  class=ConversationEntryPointNode
  outputs [#2]
  properties {entryTag {tagName=Conversation.Entry.ChillCat}}
 }
 Speech_C_4 {
  id=#2
  class=Speech_C
  outputs [#3, #4]
  properties {
   nodeName=""
   generalConfig {
    defaultMessage=My pet peeve: people who tell me who or what I should be. And jazz. I don't get jazz.
    contextualMessages [
     {
      message=Keep it calico.
      contextRequirements [{requirement {class=FortControllerRequirement_IsFirstEverConversationWithNPC}}]
      priority {curve {rowName=Tandem.Speech.Priority.FirstEverInteraction}}
     }
     {
      message="Don't worry, big bro. I'll always land on my feet."
      contextRequirements [{requirement {
         class=FortControllerRequirement_HasCID
         properties {allowedCIds [&CID_693_Athena_Commando_M_BuffCat]}
        }}]
      priority {value=0, curve {curveTable=null, rowName=None}}
     }
     {
      message=Thanks for never making me feel weird. You're the best.
      contextRequirements [{requirement {
         class=FortControllerRequirement_HasCID
         properties {allowedCIds [/BRCosmetics/Athena/Items/Cosmetics/Characters/CID_286_Athena_Commando_F_NeonCat]}
        }}]
      priority {value=0, curve {curveTable=null, rowName=None}}
     }
     {
      message=Your dad said I can show you how to kickflip if you finish your homework.
      contextRequirements [{requirement {
         class=FortControllerRequirement_HasCID
         properties {allowedCIds [/BRCosmetics/Athena/Items/Cosmetics/Characters/CID_776_Athena_Commando_M_ProfessorPup]}
        }}]
      priority {value=0, curve {curveTable=null, rowName=None}}
     }
     {
      message=Do you have anyone looking out for you?
      contextRequirements [{requirement {class=FortControllerRequirement_HasCID, properties {allowedCIds [&Character_Bites]}}}]
      priority {value=0, curve {curveTable=null, rowName=None}}
     }
     {
      message="If I'm honest, I'm a little lost too. How do I decide what to be?"
      contextRequirements [{requirement {class=FortControllerRequirement_HasCID, properties {allowedCIds [&Character_PinkSpike]}}}]
      priority {value=0, curve {curveTable=null, rowName=None}}
     }
     {
      message="You look pretty big in that suit, but I think you're hiding in it."
      contextRequirements [{requirement {class=FortControllerRequirement_HasCID, properties {allowedCIds [&Character_BadBear]}}}]
      priority {value=0, curve {curveTable=null, rowName=None}}
     }
     {
      message="You should get gone, before my brother sees you."
      contextRequirements [{requirement {
         class=FortControllerRequirement_HasCID
         properties {allowedCIds [/BRCosmetics/Athena/Items/Cosmetics/Characters/CID_419_Athena_Commando_M_Pug]}
        }}]
      priority {value=0, curve {curveTable=null, rowName=None}}
     }
    ]
   }
  }
 }
 ConversationLinkNode_0 {id=#4, class=ConversationLinkNode, properties {remoteEntryTag {tagName=Conversation.Entry.BasicServices}}}
 ConversationLinkNode_1 {id=#3, class=ConversationLinkNode, properties {remoteEntryTag {tagName=Conversation.Entry.NPCLibrary.Quests}}}
]
